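# Masked, default-lit material for the industrial hose prop, named for flowing paint, with purple color constants.
# A panning water normal map, on tiled UVs with a power-curved V for uneven flow, feeds an inverted Fresnel into Bayer dithering.
# That Fresnel also combines with a U edge falloff, sphere mask and squared V gradient to form a shaping mask.
# The normal plus Jiggle_WPO, scaled by WPO_Strength, builds the vertex wobble.
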Material /Hera/Props/Industrial/Industrial_Hoses/Industrial_Hose_A/Materials/M_Industrial_Hose_Paint_Flowing
graph {
 properties {
  bAutomaticallySetUsageInEditor=false
  blendMode=BLEND_Masked
  previewMesh=/Game/Effects/Fort_Effects/Mesh/Water/S_FX_GeyserCone_NoCollision
  usedShadingModels=Default Lit
 }
 nodes [
  Constant3Vector() {constant=<rgba 0.3195 0.0672 0.3333 1>}
  Constant3Vector() {constant=<rgba 0.0885 0.0067 0.0583 1>}
  Constant() {r=1}
  DepthFade()
  n1 = ScalarParameter() {defaultValue=0.5, parameterName=XTile}
  n2 = ScalarParameter() {defaultValue=0.5, parameterName=YTile}
  n3 = TextureCoordinate()
  n4 = ComponentMask(None=@n3) {r=true}
  n5 = Power(Base=OneMinus(None=ComponentMask(None=@n3) {g=true}))
  n6 = Panner(
   Coordinate=Multiply(A=AppendVector(A=@n1, B=@n2), B=AppendVector(A=@n4, B=@n5))
   Speed=AppendVector(A=ScalarParameter() {defaultValue=0.02, parameterName=XSpeed}, B=ScalarParameter() {defaultValue=-0.25, parameterName=YSpeed})
  ) {speedY=-0.25}
  n7 = TextureSample(UVs=@n6) {samplerType=SAMPLERTYPE_Normal, texture=/Game/Global/Textures/Default/Tileable_Patterns/T_WaterSoft_N}
  n8 = OneMinus(None=Fresnel(ExponentIn=ScalarParameter() {defaultValue=0.35, parameterName=ExponentIn}, Normal=@n7.RGB))
  MaterialFunctionCall!(In=@n8, Iterations, Scale=ScalarParameter() {defaultValue=4, parameterName="Scale (S)"}) {materialFunction=/Game/Packages/Fortress_SharedMaterials/Base_Material_Functions/MF_DitherBayerNoTex}
  n9 = ScalarParameter() {defaultValue=3, parameterName="Speed (S)"}
  n10 = ScalarParameter() {defaultValue=3, parameterName="SmallWaveScle (S)"}
  n11 = ScalarParameter() {defaultValue=5, parameterName="WaveScale (S)"}
  Multiply(
   A=Add(
    A=@n7.RGB
    B=MaterialFunctionCall!(Speed=@n9, SmallWaveScle=@n10, WaveScale=@n11) {materialFunction=/Game/Effects/Fort_Effects/Materials/_Functions/Jiggle_WPO}.Result
   )
   B=VectorParameter() {parameterName=WPO_Strength, defaultValue=<rgba 0.5 1 1 1>}.RGB
  )
  Multiply()
  n12 = ComponentMask(None=TextureCoordinate()) {g=true}
  n13 = TextureCoordinate() {uTiling=2}
  n14 = ComponentMask(None=VectorParameter() {defaultValue=<rgba 0.5 0.75 0 1>}.RGB) {g=true, r=true}
  n15 = ScalarParameter() {defaultValue=0.25, parameterName=XOffset}
  n16 = ScalarParameter() {parameterName=YOffset}
  n17 = MaterialFunctionCall(UV Channel=Add(A=TextureCoordinate(), B=AppendVector(A=@n15, B=@n16))) {materialFunction=&LinearGradient}
  n18 = OneMinus(None=@n17.UGradient)
  n19 = Saturate(None=Multiply(A=Multiply(A=@n17.UGradient, B=@n18)) {constB=4})
  n20 = Multiply(A=@n19, B=@n8)
  n21 = Multiply(A=@n7.RGB) {constB=100}
  Multiply(
   A=Multiply(A=@n12, B=@n12)
   B=Multiply(A=SphereMask(A=@n13, B=@n14) {attenuationRadius=0.5, hardnessPercent=0.5}, B=Multiply(A=@n20, B=@n21))
  )
  Power(Base=@n19) {constExponent=0.5}
  ScalarParameter()
  TextureSample(UVs=@n6) {samplerType=SAMPLERTYPE_Normal, texture=/Game/Environments/Apollo/Sets/Terrain/Textures/T_Apollo_Dirt2_N}
 ]
}
